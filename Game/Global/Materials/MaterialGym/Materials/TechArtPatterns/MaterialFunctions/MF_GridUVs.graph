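# MF_GridUVs is a material function that divides UV space into a grid, with inputs for UV, Tiling and Offset.
# It multiplies the UVs by Tiling, then adds Offset, so the offset is measured in cell units.
# GridUV floors that result and divides by Tiling, giving one stepped coordinate per cell.
# GridUV_Tiling takes the fractional part, giving coordinates that repeat from 0 to 1 inside each cell.
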MaterialFunction /Game/Global/Materials/MaterialGym/Materials/TechArtPatterns/MaterialFunctions/MF_GridUVs
graph {nodes [
  n1 = TextureCoordinate()
  n2 = FunctionInput(Preview=Constant() {r=10}) {bUsePreviewValueAsDefault=true, inputName=Tiling, inputType=FunctionInput_Scalar}
  n3 = Add(
   A=Multiply(A=FunctionInput(Preview=@n1) {bUsePreviewValueAsDefault=true, inputName=UV, inputType=FunctionInput_Vector2}, B=@n2)
   B=FunctionInput(Preview=Constant()) {bUsePreviewValueAsDefault=true, inputName=Offset, inputType=FunctionInput_Vector2}
  )
  n4 = Floor(None=@n3)
  FunctionOutput(None=NamedRerouteDeclaration(None=Divide(A=@n4, B=@n2)) {name=SquareUV, nodeColor=<rgba 0 1 0.5547 1>}) {outputName=GridUV}
  FunctionOutput(None=NamedRerouteDeclaration(None=Frac(None=@n3)) {name=SquareUV_Tiling, nodeColor=<rgba 1 0.9375 0 1>}) {outputName=GridUV_Tiling}
 ]}
