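# Material function that builds pattern UVs from a UV-encoded texture, by default a hexagon map, and outputs them alongside the base UVs.
# Base UVs are texture coordinates with an optional time-animated offset, multiplied by tiling and optionally shifted by BumpOffset parallax.
# Their fractional part samples the texture using explicit pre-tiling derivatives and clamped addressing.
# The sampled RG is added to the floored tile index, then divided to form the texture-based UV.
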MaterialFunction /Game/Global/Materials/MaterialGym/Materials/TechArtPatterns/MaterialFunctions/MF_TextureUVs
graph {nodes [
  FunctionOutput(None=NamedRerouteUsage()) {outputName=TextureUVs}
  FunctionOutput(None=NamedRerouteUsage()) {outputName=BaseUVs, sortPriority=1}
  n1 = Time()
  n2 = FunctionInput() {bUsePreviewValueAsDefault=true, inputName=OffsetXY, inputType=FunctionInput_Vector2, sortPriority=2}
  n3 = StaticBool()
  n4 = Add(
   A=TextureCoordinate()
   B=StaticSwitch(
    True=Multiply(A=@n1, B=@n2)
    False=@n2
    Value=FunctionInput(Preview=@n3) {bUsePreviewValueAsDefault=true, inputName=AnimateOffset, inputType=FunctionInput_StaticBool, sortPriority=3}
   )
  )
  n5 = FunctionInput() {bUsePreviewValueAsDefault=true, inputName=TilingXY, inputType=FunctionInput_Vector2, sortPriority=1, previewValue=<1 1 0 1>}
  n6 = Multiply(A=@n4, B=NamedRerouteDeclaration(None=@n5) {name=PatternTiling, nodeColor=<rgba 1 0 0.8203 1>})
  n7 = FunctionInput() {bUsePreviewValueAsDefault=true, inputName=Height, inputType=FunctionInput_Scalar, sortPriority=10}
  n8 = StaticBool()
  n9 = NamedRerouteDeclaration(None=StaticSwitch(
    True=BumpOffset(Coordinate=@n6, Height=@n7)
    False=@n6
    Value=FunctionInput(Preview=@n8) {bUsePreviewValueAsDefault=true, inputName=UseParallax, inputType=FunctionInput_StaticBool, sortPriority=9}
   )) {name=BaseUVs, nodeColor=<rgba 0.9922 1 0 1>}
  n10 = Frac(None=@n9)
  n11 = FunctionInput(Preview=TextureObject() {samplerType=SAMPLERTYPE_LinearColor, texture=/Game/Global/Materials/MaterialGym/Materials/TechArtPatterns/Textures/T_hexagons_UV}) {bUsePreviewValueAsDefault=true, inputName=Texture, inputType=FunctionInput_Texture2D}
  n12 = DDX(Value=@n4)
  n13 = DDY(Value=@n4)
  n14 = Add(
   A=ComponentMask(None=TextureSample!(UVs=@n10, Tex=@n11, "DDX(UVs)"=@n12, "DDY(UVs)"=@n13, Apply View MipBias) {mipValueMode=TMVM_Derivative, samplerSource=SSM_Clamp_WorldGroupSettings, samplerType=SAMPLERTYPE_LinearColor}.RGB) {g=true, r=true}
   B=Floor(None=@n9)
  )
  NamedRerouteDeclaration(None=Divide(A=Subtract(A=@n14) {constB=0}, B=NamedRerouteUsage())) {name=TextureBasedUV, desc=Texture based UV, nodeColor=<rgba 0 0.4844 1 1>}
 ]}
